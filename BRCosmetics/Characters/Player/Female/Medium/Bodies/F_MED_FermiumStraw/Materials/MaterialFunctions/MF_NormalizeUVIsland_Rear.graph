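# Material function for the FermiumStraw female body that normalizes rear UVs, apparently treating the Rear and Neck islands as one island joined across a strap gap.
# An If shifts UVs on one side of a V threshold, and a combined size spans rear, neck and strap.
# It outputs the normalized UV and a 1:1 aspect version using the larger dimension.
# Each has a 0-1 bounds mask built from two multiplied one-dimensional SphereMasks.
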MaterialFunction /BRCosmetics/Characters/Player/Female/Medium/Bodies/F_MED_FermiumStraw/Materials/MaterialFunctions/MF_NormalizeUVIsland_Rear
graph {nodes [
  n1 = Constant() {r=0.5}
  n2 = Subtract(A=NamedRerouteUsage(), B=NamedRerouteUsage())
  n3 = NamedRerouteUsage()
  n4 = MaterialFunctionCall(Float2=FunctionOutput(None=Divide(A=@n2, B=@n3)) {outputName="Island UV (V2)"}) {materialFunction=&BreakOutFloat2Components}
  FunctionOutput(None=Multiply(A=SphereMask(A=@n1, B=@n4.R) {attenuationRadius=0.5}, B=SphereMask(A=@n4.G, B=@n1) {attenuationRadius=0.5})) {outputName="Island 0-1 Mask (S)", sortPriority=5}
  n5 = Constant() {r=0.5}
  n6 = Subtract(A=NamedRerouteUsage(), B=NamedRerouteUsage())
  n7 = Max(A=NamedRerouteUsage(), B=NamedRerouteUsage())
  n8 = MaterialFunctionCall(Float2=FunctionOutput(None=Divide(A=@n6, B=@n7)) {outputName="Island UV - 1:1 Aspect Ratio (V2)", sortPriority=10}) {materialFunction=&BreakOutFloat2Components}
  FunctionOutput(None=Multiply(A=SphereMask(A=@n5, B=@n8.R) {attenuationRadius=0.5}, B=SphereMask(A=@n8.G, B=@n5) {attenuationRadius=0.5})) {outputName="Island UV - 1:1 Aspect Ratio 0-1 Mask (S)", sortPriority=15}
  NamedRerouteDeclaration(None=FunctionInput() {bUsePreviewValueAsDefault=true, inputName=Rear UV Min, inputType=FunctionInput_Vector2, sortPriority=10}) {name="Rear UV Min (V2)", nodeColor=<rgba 0 0.2734 1 1>}
  MaterialExpressionNamedRerouteDeclaration_11 = NamedRerouteDeclaration(None=Subtract(A=NamedRerouteUsage(), B=NamedRerouteUsage())) {name="Neck UV Size (V2)", nodeColor=<rgba 0 0.6953 1 1>}
  n9 = Convert(None=@MaterialExpressionNamedRerouteDeclaration_11) {
   nodeName=Break Float2
   convertInputs [{
     type=Vector2
     expressionInput {
      expression=&self:MaterialExpressionNamedRerouteDeclaration_11
      inputName=None
      mask=0
      maskA=0
      maskB=0
      maskG=0
      maskR=0
      outputIndex=0
     }
    }]
   convertOutputs [{type=Scalar}, {type=Scalar}]
  }
  NamedRerouteDeclaration(None=@n9) {name="Neck V Size (S)", nodeColor=<rgba 0.3125 1 0 1>}
  MaterialExpressionNamedRerouteDeclaration_0 = NamedRerouteDeclaration(None=FunctionInput(Preview=TextureCoordinate()) {bUsePreviewValueAsDefault=true, inputName=UV in, inputType=FunctionInput_Vector2}) {name="UV in (V2)", nodeColor=<rgba 1 0.2087 0 1>}
  n10 = Convert(None=@MaterialExpressionNamedRerouteDeclaration_0) {
   nodeName=Break Float2
   convertInputs [{
     type=Vector2
     expressionInput {
      expression=&self:MaterialExpressionNamedRerouteDeclaration_0
      inputName=None
      mask=0
      maskA=0
      maskB=0
      maskG=0
      maskR=0
      outputIndex=0
     }
    }]
   convertOutputs [{type=Scalar}, {type=Scalar}]
  }
  NamedRerouteDeclaration(None=@n10) {name="U in (S)", nodeColor=<rgba 0 0.5565 1 1>}
  NamedRerouteDeclaration(None=@n10) {name="V in (S)", nodeColor=<rgba 0.3125 1 0 1>}
  n11 = FunctionInput() {bUsePreviewValueAsDefault=true, inputName=Rear UV Max, inputType=FunctionInput_Vector2, sortPriority=20, previewValue=<1 1 0 1>}
  NamedRerouteDeclaration(None=@n11) {name="Rear UV Max (V2)", nodeColor=<rgba 0 1 0.3478 1>}
  n12 = NamedRerouteUsage()
  n13 = NamedRerouteUsage()
  MaterialExpressionNamedRerouteDeclaration_18 = NamedRerouteDeclaration(None=Multiply(A=Add(A=@n12, B=@n13)) {constB=0.5}) {name="Rear UV Center (V2)", nodeColor=<rgba 1 0 0.6797 1>}
  n14 = Convert(None=@MaterialExpressionNamedRerouteDeclaration_18) {
   nodeName=Break Float2
   convertInputs [{
     type=Vector2
     expressionInput {
      expression=&self:MaterialExpressionNamedRerouteDeclaration_18
      inputName=None
      mask=0
      maskA=0
      maskB=0
      maskG=0
      maskR=0
      outputIndex=0
     }
    }]
   convertOutputs [{type=Scalar}, {type=Scalar}]
  }
  NamedRerouteDeclaration(None=@n14) {name="Rear U Center (S)", nodeColor=<rgba 0 0.6957 1 1>}
  NamedRerouteDeclaration(None=@n14) {name="Rear V Center (S)", nodeColor=<rgba 0.3594 0 1 1>}
  n15 = NamedRerouteUsage()
  n16 = NamedRerouteUsage()
  MaterialExpressionNamedRerouteDeclaration_19 = NamedRerouteDeclaration(None=Multiply(A=Add(A=@n15, B=@n16)) {constB=0.5}) {name="Neck UV Center (V2)", nodeColor=<rgba 1 0 0.6797 1>}
  n17 = Convert(None=@MaterialExpressionNamedRerouteDeclaration_19) {
   nodeName=Break Float2
   convertInputs [{
     type=Vector2
     expressionInput {
      expression=&self:MaterialExpressionNamedRerouteDeclaration_19
      inputName=None
      mask=0
      maskA=0
      maskB=0
      maskG=0
      maskR=0
      outputIndex=0
     }
    }]
   convertOutputs [{type=Scalar}, {type=Scalar}]
  }
  NamedRerouteDeclaration(None=@n17) {name="Neck U Center (S)", nodeColor=<rgba 0 0.6957 1 1>}
  NamedRerouteDeclaration(None=@n17) {name="Neck V Center (S)", nodeColor=<rgba 0.3594 0 1 1>}
  n18 = NamedRerouteUsage()
  n19 = Add(A=Add(A=NamedRerouteUsage(), B=NamedRerouteUsage()), B=NamedRerouteUsage())
  NamedRerouteDeclaration(None=AppendVector(
    A=NamedRerouteDeclaration(None=@n18) {name="Final U Size (S)", nodeColor=<rgba 0.0078 0 1 1>}
    B=NamedRerouteDeclaration(None=@n19) {name="Final V Size (S)", nodeColor=<rgba 0.3125 1 0 1>}
   )) {name="Final UV Size (V2)", nodeColor=<rgba 0 0.6953 1 1>}
  NamedRerouteDeclaration(None=FunctionInput() {bUsePreviewValueAsDefault=true, inputName=Neck UV Min, inputType=FunctionInput_Vector2, sortPriority=30}) {name="Neck UV Min (V2)", nodeColor=<rgba 0 0.2734 1 1>}
  MaterialExpressionAdd_7 = Add(A=NamedRerouteUsage(), B=NamedRerouteUsage())
  n20 = Convert(None=@MaterialExpressionAdd_7) {
   convertInputs [{
     type=Scalar
     expressionInput {
      expression=&self:MaterialExpressionAdd_7
      inputName=None
      mask=0
      maskA=0
      maskB=0
      maskG=0
      maskR=0
      outputIndex=0
     }
    }]
   convertOutputs [{type=Vector2}]
  }
  NamedRerouteDeclaration(None=Subtract(A=NamedRerouteUsage(), B=@n20)) {name="Final UV Min (V2)", nodeColor=<rgba 0 0.2734 1 1>}
  n21 = FunctionInput() {bUsePreviewValueAsDefault=true, inputName=Strap V Size, inputType=FunctionInput_Scalar, sortPriority=50, previewValue=<0 1 0 1>}
  NamedRerouteDeclaration(None=@n21) {name="Strap V Size (S)", nodeColor=<rgba 0.3125 1 0 1>}
  n22 = NamedRerouteUsage()
  n23 = Multiply(A=Subtract(A=NamedRerouteUsage(), B=NamedRerouteUsage())) {constB=0.5}
  NamedRerouteDeclaration(None=AppendVector(A=Add(A=@n22, B=@n23), B=NamedRerouteUsage())) {name="Centered UV (V2)", nodeColor=<rgba 0 1 0.7656 1>}
  MaterialExpressionNamedRerouteDeclaration_9 = NamedRerouteDeclaration(None=Subtract(A=NamedRerouteUsage(), B=NamedRerouteUsage())) {name="Rear UV Size (V2)", nodeColor=<rgba 0 0.6953 1 1>}
  n24 = Convert(None=@MaterialExpressionNamedRerouteDeclaration_9) {
   nodeName=Break Float2
   convertInputs [{
     type=Vector2
     expressionInput {
      expression=&self:MaterialExpressionNamedRerouteDeclaration_9
      inputName=None
      mask=0
      maskA=0
      maskB=0
      maskG=0
      maskR=0
      outputIndex=0
     }
    }]
   convertOutputs [{type=Scalar}, {type=Scalar}]
  }
  NamedRerouteDeclaration(None=@n24) {name="Rear U Size (S)", nodeColor=<rgba 0.0078 0 1 1>}
  NamedRerouteDeclaration(None=@n24) {name="Rear V Size (S)", nodeColor=<rgba 0.3125 1 0 1>}
  n25 = FunctionInput() {bUsePreviewValueAsDefault=true, inputName=Neck UV Max, inputType=FunctionInput_Vector2, sortPriority=40, previewValue=<1 1 0 1>}
  NamedRerouteDeclaration(None=@n25) {name="Neck UV Max (V2)", nodeColor=<rgba 0 1 0.3478 1>}
  NamedRerouteDeclaration(None=@n9) {name="Neck U Size (S)", nodeColor=<rgba 0.0078 0 1 1>}
  n26 = NamedRerouteUsage()
  n27 = NamedRerouteUsage()
  n28 = NamedRerouteUsage()
  n29 = NamedRerouteUsage()
  n30 = ComponentMask(None=NamedRerouteUsage()) {g=true}
  n31 = Subtract(A=AppendVector(A=NamedRerouteUsage(), B=ComponentMask(None=@n28) {g=true}), B=AppendVector(A=NamedRerouteUsage(), B=Add(A=@n29, B=@n30)))
  MaterialExpressionNamedRerouteDeclaration_24 = NamedRerouteDeclaration(None=If(A=NamedRerouteUsage(), B=ComponentMask(None=@n26) {g=true}, "A > B"=@n27, "A < B"=Add(A=@n27, B=@n31))) {name="Final UV in (V2)", nodeColor=<rgba 1 0.2087 0 1>}
  n32 = Convert(None=@MaterialExpressionNamedRerouteDeclaration_24) {
   nodeName=Break Float2
   convertInputs [{
     type=Vector2
     expressionInput {
      expression=&self:MaterialExpressionNamedRerouteDeclaration_24
      inputName=None
      mask=0
      maskA=0
      maskB=0
      maskG=0
      maskR=0
      outputIndex=0
     }
    }]
   convertOutputs [{type=Scalar}, {type=Scalar}]
  }
  NamedRerouteDeclaration(None=@n32) {name="Final U (S)", nodeColor=<rgba 0 0.5565 1 1>}
  NamedRerouteDeclaration(None=@n32) {name="Final V (S)", nodeColor=<rgba 0.3125 1 0 1>}
 ]}
