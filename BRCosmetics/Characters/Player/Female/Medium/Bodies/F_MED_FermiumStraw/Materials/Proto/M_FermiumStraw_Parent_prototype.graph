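# Prototype parent material for the FermiumStraw female body cosmetic, adding a customizable projected loading-screen design to the outfit.
# Channel masks separate front, rear/neck and sleeve regions, whose UV islands are normalized to 1:1 aspect.
# Per-region transforms and user offsets drive shared UV offset, rotation, scale and optional clamping, with a forced-symmetry sleeve control.
# The customized layer blends into material attributes between the shared CharacterParent PreFX and PostFX functions.
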Material /BRCosmetics/Characters/Player/Female/Medium/Bodies/F_MED_FermiumStraw/Materials/Proto/M_FermiumStraw_Parent_prototype
graph {
 expressions {
  MaterialExpressionAdd_1 {
   class=MaterialExpressionAdd
   inputs [
    A {from=MaterialExpressionDivide_8}
    B {from=""}
   ]
   properties {constB=0.5}
  }
  MaterialExpressionAdd_2 {
   class=MaterialExpressionAdd
   inputs [
    A {from=MaterialExpressionMultiply_12}
    B {from=MaterialExpressionMultiply_0}
   ]
  }
  MaterialExpressionAdd_3 {
   class=MaterialExpressionAdd
   inputs [
    A {from=MaterialExpressionConvert_5}
    B {from=MaterialExpressionReroute_14}
   ]
  }
  MaterialExpressionAdd_4 {
   class=MaterialExpressionAdd
   inputs [
    A {from=MaterialExpressionVectorParameter_10, output=B}
    B {from=""}
   ]
   properties {constB=0.25}
  }
  MaterialExpressionAdd_5 {
   class=MaterialExpressionAdd
   inputs [
    A {from=MaterialExpressionNamedRerouteUsage_33}
    B {from=MaterialExpressionVectorParameter_11, output=R}
   ]
  }
  MaterialExpressionAdd_6 {
   class=MaterialExpressionAdd
   inputs [
    A {from=MaterialExpressionNamedRerouteUsage_43}
    B {from=MaterialExpressionVectorParameter_12, output=R}
   ]
  }
  MaterialExpressionAdd_7 {
   class=MaterialExpressionAdd
   inputs [
    A {from=MaterialExpressionNamedRerouteUsage_49}
    B {from=MaterialExpressionVectorParameter_10, output=R}
   ]
  }
  MaterialExpressionAppendVector_0 {
   class=MaterialExpressionAppendVector
   inputs [
    A {from=MaterialExpressionVectorParameter_2, output=U Min}
    B {from=MaterialExpressionVectorParameter_2, output=U Min}
   ]
  }
  MaterialExpressionAppendVector_1 {
   class=MaterialExpressionAppendVector
   inputs [
    A {from=MaterialExpressionVectorParameter_2, output=U Max}
    B {from=MaterialExpressionVectorParameter_2, output=U Max}
   ]
  }
  MaterialExpressionAppendVector_10 {
   class=MaterialExpressionAppendVector
   inputs [
    A {from=MaterialExpressionVectorParameter_8, output=U Min}
    B {from=MaterialExpressionVectorParameter_8, output=U Min}
   ]
  }
  MaterialExpressionAppendVector_11 {
   class=MaterialExpressionAppendVector
   inputs [
    A {from=MaterialExpressionVectorParameter_8, output=U Max}
    B {from=MaterialExpressionVectorParameter_8, output=U Max}
   ]
  }
  MaterialExpressionAppendVector_2 {
   class=MaterialExpressionAppendVector
   inputs [
    A {from=MaterialExpressionLinearInterpolate_3, output=""}
    B {from=MaterialExpressionOneMinus_9, output=""}
   ]
  }
  MaterialExpressionAppendVector_4 {
   class=MaterialExpressionAppendVector
   inputs [
    A {from=MaterialExpressionVectorParameter_0, output=U Min}
    B {from=MaterialExpressionVectorParameter_0, output=U Min}
   ]
  }
  MaterialExpressionAppendVector_5 {
   class=MaterialExpressionAppendVector
   inputs [
    A {from=MaterialExpressionVectorParameter_0, output=U Max}
    B {from=MaterialExpressionVectorParameter_0, output=U Max}
   ]
  }
  MaterialExpressionAppendVector_6 {
   class=MaterialExpressionAppendVector
   inputs [
    A {from=MaterialExpressionVectorParameter_1, output=U Min}
    B {from=MaterialExpressionVectorParameter_1, output=U Min}
   ]
  }
  MaterialExpressionAppendVector_7 {
   class=MaterialExpressionAppendVector
   inputs [
    A {from=MaterialExpressionVectorParameter_1, output=U Max}
    B {from=MaterialExpressionVectorParameter_1, output=U Max}
   ]
  }
  MaterialExpressionAppendVector_8 {
   class=MaterialExpressionAppendVector
   inputs [
    A {from=MaterialExpressionVectorParameter_5, output=U Min}
    B {from=MaterialExpressionVectorParameter_5, output=U Min}
   ]
  }
  MaterialExpressionAppendVector_9 {
   class=MaterialExpressionAppendVector
   inputs [
    A {from=MaterialExpressionVectorParameter_5, output=U Max}
    B {from=MaterialExpressionVectorParameter_5, output=U Max}
   ]
  }
  MaterialExpressionBlendMaterialAttributes_2 {
   class=MaterialExpressionBlendMaterialAttributes
   inputs [
    A {from=MaterialExpressionNamedRerouteDeclaration_13, output=""}
    B {from=MaterialExpressionReroute_5, output=""}
    Alpha {from=MaterialExpressionNamedRerouteUsage_8, output=""}
   ]
   properties {}
  }
  MaterialExpressionChannelMaskParameter_1 {
   class=MaterialExpressionChannelMaskParameter
   inputs [None {from=MaterialExpressionNamedRerouteUsage_1, output=""}]
   properties {group=FermiumStraw | 00_Base Masks, maskChannel=Green, parameterName=TechArtMask_OutfitChannel, sortPriority=10, defaultValue=<rgba 0 1 0 0>}
  }
  MaterialExpressionChannelMaskParameter_2 {
   class=MaterialExpressionChannelMaskParameter
   inputs [None {from=MaterialExpressionNamedRerouteUsage_1, output=""}]
   properties {group=FermiumStraw | 00_Base Masks, parameterName=TechArtMask_FabricChannel, sortPriority=20}
  }
  MaterialExpressionChannelMaskParameter_4 {
   class=MaterialExpressionChannelMaskParameter
   inputs [None {from=MaterialExpressionTextureSampleParameter2D_3, output=RGB}]
   properties {group=FermiumStraw | 00_Base Masks, parameterName=ProjectionMask_SleevesChannel, sortPriority=70}
  }
  MaterialExpressionChannelMaskParameter_7 {
   class=MaterialExpressionChannelMaskParameter
   inputs [None {from=MaterialExpressionTextureSampleParameter2D_3, output=RGB}]
   properties {
    group=FermiumStraw | 00_Base Masks
    maskChannel=Green
    parameterName=ProjectionMask_RearChannel
    sortPriority=60
    defaultValue=<rgba 0 1 0 0>
   }
  }
  MaterialExpressionChannelMaskParameter_8 {
   class=MaterialExpressionChannelMaskParameter
   inputs [None {from=MaterialExpressionTextureSampleParameter2D_3, output=RGB}]
   properties {group=FermiumStraw | 00_Base Masks, maskChannel=Blue, parameterName=ProjectionMask_NeckChannel, sortPriority=80, defaultValue=<rgba 0 0 1 0>}
  }
  MaterialExpressionChannelMaskParameter_9 {
   class=MaterialExpressionChannelMaskParameter
   inputs [None {from=MaterialExpressionTextureSampleParameter2D_4, output=G}]
   properties {group=FermiumStraw | 00_Base Masks, maskChannel=Green, parameterName=ExtraMask_PartsChannel, sortPriority=40, defaultValue=<rgba 0 1 0 0>}
  }
  MaterialExpressionComponentMask_5 {class=MaterialExpressionComponentMask, inputs [None {from=MaterialExpressionLocalPosition_1, output=XYZ}], properties {r=true}}
  MaterialExpressionComponentMask_9 {class=MaterialExpressionComponentMask, inputs [None {from=MaterialExpressionNamedRerouteUsage_57}], properties {g=true}}
  MaterialExpressionConstant2Vector_5 {class=MaterialExpressionConstant2Vector, properties {g=1, r=2}}
  MaterialExpressionConstant_0 {class=MaterialExpressionConstant, properties {r=-3}}
  MaterialExpressionConstant_11 {class=MaterialExpressionConstant, properties {r=0.5}}
  MaterialExpressionConstant_2 {class=MaterialExpressionConstant, properties {r=2}}
  MaterialExpressionConstant_3 {class=MaterialExpressionConstant, properties {r=1}}
  MaterialExpressionConvert_10 {
   class=MaterialExpressionConvert
   inputs [
    None {from=MaterialExpressionScalarParameter_13}
    None {from=MaterialExpressionScalarParameter_14}
    None {from=MaterialExpressionScalarParameter_15}
   ]
   properties {
    nodeName=Make Vector3
    convertInputs [
     {
      type=Scalar
      expressionInput {
       expression=&self:MaterialExpressionScalarParameter_13
       inputName=None
       mask=0
       maskA=0
       maskB=0
       maskG=0
       maskR=0
       outputIndex=0
      }
     }
     {
      type=Scalar
      expressionInput {
       expression=&self:MaterialExpressionScalarParameter_14
       inputName=None
       mask=0
       maskA=0
       maskB=0
       maskG=0
       maskR=0
       outputIndex=0
      }
     }
     {
      type=Scalar
      expressionInput {
       expression=&self:MaterialExpressionScalarParameter_15
       inputName=None
       mask=0
       maskA=0
       maskB=0
       maskG=0
       maskR=0
       outputIndex=0
      }
     }
    ]
    convertOutputs [{type=Vector3}]
   }
  }
  MaterialExpressionConvert_11 {
   class=MaterialExpressionConvert
   inputs [None {from=MaterialExpressionLinearInterpolate_18}]
   properties {
    nodeName=Break Float4
    convertInputs [{
      type=Vector4
      defaultValue=<rgba 0 0 0 1>
      expressionInput {
       expression=&self:MaterialExpressionLinearInterpolate_18
       inputName=None
       mask=0
       maskA=0
       maskB=0
       maskG=0
       maskR=0
       outputIndex=0
      }
     }]
    convertOutputs [{type=Scalar}, {type=Scalar}, {type=Scalar}, {type=Scalar}]
   }
  }
  MaterialExpressionConvert_12 {
   class=MaterialExpressionConvert
   inputs [None {from=MaterialExpressionMultiply_5}]
   properties {
    nodeName=Break Float3
    convertInputs [{
      type=Vector3
      expressionInput {
       expression=&self:MaterialExpressionMultiply_5
       inputName=None
       mask=0
       maskA=0
       maskB=0
       maskG=0
       maskR=0
       outputIndex=0
      }
     }]
    convertOutputs [{type=Scalar}, {type=Scalar}, {type=Scalar}]
   }
  }
  MaterialExpressionConvert_13 {
   class=MaterialExpressionConvert
   inputs [None {from=MaterialExpressionNamedRerouteUsage_17}]
   properties {
    nodeName=Break Float2
    convertInputs [{
      type=Vector2
      expressionInput {
       expression=&self:MaterialExpressionNamedRerouteUsage_17
       inputName=None
       mask=0
       maskA=0
       maskB=0
       maskG=0
       maskR=0
       outputIndex=0
      }
     }]
    convertOutputs [{type=Scalar}, {type=Scalar}]
   }
  }
  MaterialExpressionConvert_14 {
   class=MaterialExpressionConvert
   inputs [
    None {from=MaterialExpressionScalarParameter_18}
    None {from=MaterialExpressionScalarParameter_19}
    None {from=MaterialExpressionScalarParameter_20}
   ]
   properties {
    nodeName=Make Vector3
    convertInputs [
     {
      type=Scalar
      expressionInput {
       expression=&self:MaterialExpressionScalarParameter_18
       inputName=None
       mask=0
       maskA=0
       maskB=0
       maskG=0
       maskR=0
       outputIndex=0
      }
     }
     {
      type=Scalar
      expressionInput {
       expression=&self:MaterialExpressionScalarParameter_19
       inputName=None
       mask=0
       maskA=0
       maskB=0
       maskG=0
       maskR=0
       outputIndex=0
      }
     }
     {
      type=Scalar
      expressionInput {
       expression=&self:MaterialExpressionScalarParameter_20
       inputName=None
       mask=0
       maskA=0
       maskB=0
       maskG=0
       maskR=0
       outputIndex=0
      }
     }
    ]
    convertOutputs [{type=Vector3}]
   }
  }
  MaterialExpressionConvert_15 {
   class=MaterialExpressionConvert
   inputs [None {from=MaterialExpressionSubtract_3}]
   properties {
    convertInputs [{
      type=Scalar
      expressionInput {
       expression=&self:MaterialExpressionSubtract_3
       inputName=None
       mask=0
       maskA=0
       maskB=0
       maskG=0
       maskR=0
       outputIndex=0
      }
     }]
    convertOutputs [{type=Vector3, defaultValue=<rgba 0 1 1 1>}]
   }
  }
  MaterialExpressionConvert_16 {
   class=MaterialExpressionConvert
   inputs [
    None {from=MaterialExpressionReroute_7}
    None {from=MaterialExpressionVectorParameter_10, output=G}
    None {from=MaterialExpressionFmod_0}
    None {from=MaterialExpressionReroute_18}
   ]
   properties {
    nodeName=Make Vector4
    convertInputs [
     {
      type=Scalar
      expressionInput {
       expression=&self:MaterialExpressionReroute_7
       inputName=None
       mask=0
       maskA=0
       maskB=0
       maskG=0
       maskR=0
       outputIndex=0
      }
     }
     {
      type=Scalar
      expressionInput {
       expression=&self:MaterialExpressionVectorParameter_10
       inputName=None
       mask=1
       maskA=0
       maskB=0
       maskG=1
       maskR=0
       outputIndex=2
      }
     }
     {
      type=Scalar
      expressionInput {
       expression=&self:MaterialExpressionFmod_0
       inputName=None
       mask=0
       maskA=0
       maskB=0
       maskG=0
       maskR=0
       outputIndex=0
      }
     }
     {
      type=Scalar
      expressionInput {
       expression=&self:MaterialExpressionReroute_18
       inputName=None
       mask=0
       maskA=0
       maskB=0
       maskG=0
       maskR=0
       outputIndex=0
      }
     }
    ]
    convertOutputs [{type=Vector4, defaultValue=<rgba 0 0 0 1>}]
   }
  }
  MaterialExpressionConvert_17 {
   class=MaterialExpressionConvert
   inputs [None {from=MaterialExpressionNamedRerouteUsage_20}]
   properties {
    convertInputs [{
      type=Scalar
      expressionInput {
       expression=&self:MaterialExpressionNamedRerouteUsage_20
       inputName=None
       mask=0
       maskA=0
       maskB=0
       maskG=0
       maskR=0
       outputIndex=0
      }
     }]
    convertOutputs [{type=Vector3, defaultValue=<rgba 0 -1 1 1>}]
   }
  }
  MaterialExpressionConvert_18 {
   class=MaterialExpressionConvert
   inputs [
    None {from=MaterialExpressionNamedRerouteUsage_21}
    None {from=MaterialExpressionNamedRerouteUsage_32}
   ]
   properties {
    convertInputs [
     {
      type=Scalar
      expressionInput {
       expression=&self:MaterialExpressionNamedRerouteUsage_21
       inputName=None
       mask=0
       maskA=0
       maskB=0
       maskG=0
       maskR=0
       outputIndex=0
      }
     }
     {
      type=Scalar
      expressionInput {
       expression=&self:MaterialExpressionNamedRerouteUsage_32
       inputName=None
       mask=0
       maskA=0
       maskB=0
       maskG=0
       maskR=0
       outputIndex=0
      }
     }
    ]
    convertOutputs [{type=Vector4, defaultValue=<rgba 0 1 1 1>}]
   }
  }
  MaterialExpressionConvert_21 {
   class=MaterialExpressionConvert
   inputs [
    None {from=MaterialExpressionLinearInterpolate_2}
    None {from=MaterialExpressionNamedRerouteUsage_28}
   ]
   properties {
    convertInputs [
     {
      type=Vector4
      defaultValue=<rgba 0 0 0 1>
      expressionInput {
       expression=&self:MaterialExpressionLinearInterpolate_2
       inputName=None
       mask=0
       maskA=0
       maskB=0
       maskG=0
       maskR=0
       outputIndex=0
      }
     }
     {
      type=Scalar
      expressionInput {
       expression=&self:MaterialExpressionNamedRerouteUsage_28
       inputName=None
       mask=0
       maskA=0
       maskB=0
       maskG=0
       maskR=0
       outputIndex=0
      }
     }
    ]
    convertOutputs [{type=Vector2}, {type=Vector2}, {type=Scalar}]
   }
  }
  MaterialExpressionConvert_5 {
   class=MaterialExpressionConvert
   inputs [None {from=MaterialExpressionScalarParameter_3}]
   properties {
    convertInputs [{
      type=Scalar
      expressionInput {
       expression=&self:MaterialExpressionScalarParameter_3
       inputName=None
       mask=0
       maskA=0
       maskB=0
       maskG=0
       maskR=0
       outputIndex=0
      }
     }]
    convertOutputs [{type=Vector3}]
   }
  }
  MaterialExpressionConvert_8 {
   class=MaterialExpressionConvert
   inputs [
    None {from=MaterialExpressionAdd_5}
    None {from=MaterialExpressionVectorParameter_11, output=G}
    None {from=MaterialExpressionVectorParameter_11, output=G}
    None {from=MaterialExpressionVectorParameter_11, output=G}
   ]
   properties {
    nodeName=Make Vector4
    convertInputs [
     {
      type=Scalar
      expressionInput {
       expression=&self:MaterialExpressionAdd_5
       inputName=None
       mask=0
       maskA=0
       maskB=0
       maskG=0
       maskR=0
       outputIndex=0
      }
     }
     {
      type=Scalar
      expressionInput {
       expression=&self:MaterialExpressionVectorParameter_11
       inputName=None
       mask=1
       maskA=0
       maskB=0
       maskG=1
       maskR=0
       outputIndex=2
      }
     }
     {
      type=Scalar
      expressionInput {
       expression=&self:MaterialExpressionVectorParameter_11
       inputName=None
       mask=1
       maskA=0
       maskB=1
       maskG=0
       maskR=0
       outputIndex=3
      }
     }
     {
      type=Scalar
      expressionInput {
       expression=&self:MaterialExpressionVectorParameter_11
       inputName=None
       mask=1
       maskA=1
       maskB=0
       maskG=0
       maskR=0
       outputIndex=4
      }
     }
    ]
    convertOutputs [{type=Vector4, defaultValue=<rgba 0 0 0 1>}]
   }
  }
  MaterialExpressionConvert_9 {
   class=MaterialExpressionConvert
   inputs [
    None {from=MaterialExpressionAdd_6}
    None {from=MaterialExpressionVectorParameter_12, output=G}
    None {from=MaterialExpressionVectorParameter_12, output=G}
    None {from=MaterialExpressionVectorParameter_12, output=G}
   ]
   properties {
    nodeName=Make Vector4
    convertInputs [
     {
      type=Scalar
      expressionInput {
       expression=&self:MaterialExpressionAdd_6
       inputName=None
       mask=0
       maskA=0
       maskB=0
       maskG=0
       maskR=0
       outputIndex=0
      }
     }
     {
      type=Scalar
      expressionInput {
       expression=&self:MaterialExpressionVectorParameter_12
       inputName=None
       mask=1
       maskA=0
       maskB=0
       maskG=1
       maskR=0
       outputIndex=2
      }
     }
     {
      type=Scalar
      expressionInput {
       expression=&self:MaterialExpressionVectorParameter_12
       inputName=None
       mask=1
       maskA=0
       maskB=1
       maskG=0
       maskR=0
       outputIndex=3
      }
     }
     {
      type=Scalar
      expressionInput {
       expression=&self:MaterialExpressionVectorParameter_12
       inputName=None
       mask=1
       maskA=1
       maskB=0
       maskG=0
       maskR=0
       outputIndex=4
      }
     }
    ]
    convertOutputs [{type=Vector4, defaultValue=<rgba 0 0 0 1>}]
   }
  }
  MaterialExpressionDivide_8 {
   class=MaterialExpressionDivide
   inputs [
    A {from=MaterialExpressionRotator_2}
    B {from=MaterialExpressionMultiply_19}
   ]
  }
  MaterialExpressionFmod_0 {
   class=MaterialExpressionFmod
   inputs [
    A {from=MaterialExpressionAdd_4, output=""}
    B {from=MaterialExpressionConstant_3, output=""}
   ]
   properties {}
  }
  MaterialExpressionFrac_1 {class=MaterialExpressionFrac, inputs [None {from=MaterialExpressionNamedRerouteUsage_88, output=""}], properties {}}
  MaterialExpressionGetMaterialAttributes_0 {
   class=MaterialExpressionGetMaterialAttributes
   inputs [None {from=MaterialExpressionNamedRerouteUsage_24, output=""}]
   properties {attributeGetTypes [#1]}
  }
  MaterialExpressionLinearInterpolate_0 {
   class=MaterialExpressionLinearInterpolate
   inputs [
    A {from=MaterialExpressionVectorParameter_7, output=RGB}
    B {from=MaterialExpressionVectorParameter_3, output=RGB}
    Alpha {from=MaterialExpressionScalarParameter_1}
   ]
  }
  MaterialExpressionLinearInterpolate_1 {
   class=MaterialExpressionLinearInterpolate
   inputs [
    A {from=MaterialExpressionNamedRerouteUsage_51}
    B {from=MaterialExpressionNamedRerouteUsage_48}
    Alpha {from=MaterialExpressionNamedRerouteUsage_47}
   ]
  }
  MaterialExpressionLinearInterpolate_11 {
   class=MaterialExpressionLinearInterpolate
   inputs [
    A {from=MaterialExpressionNamedRerouteUsage_15}
    B {from=MaterialExpressionNamedRerouteUsage_59}
    Alpha {from=MaterialExpressionNamedRerouteUsage_30}
   ]
  }
  MaterialExpressionLinearInterpolate_12 {
   class=MaterialExpressionLinearInterpolate
   inputs [
    A {from=MaterialExpressionMultiply_9}
    B {from=MaterialExpressionLinearInterpolate_1}
    Alpha {from=MaterialExpressionNamedRerouteUsage_9}
   ]
  }
  MaterialExpressionLinearInterpolate_13 {
   class=MaterialExpressionLinearInterpolate
   inputs [
    A {from=MaterialExpressionLinearInterpolate_9}
    B {from=MaterialExpressionNamedRerouteUsage_61}
    Alpha {from=MaterialExpressionNamedRerouteUsage_69}
   ]
  }
  MaterialExpressionLinearInterpolate_14 {
   class=MaterialExpressionLinearInterpolate
   inputs [
    A {from=MaterialExpressionLinearInterpolate_13}
    B {from=MaterialExpressionNamedRerouteUsage_62}
    Alpha {from=MaterialExpressionNamedRerouteUsage_70}
   ]
  }
  MaterialExpressionLinearInterpolate_16 {
   class=MaterialExpressionLinearInterpolate
   inputs [
    A {from=MaterialExpressionLinearInterpolate_17}
    B {from=MaterialExpressionNamedRerouteUsage_6}
    Alpha {from=MaterialExpressionNamedRerouteUsage_77}
   ]
  }
  MaterialExpressionLinearInterpolate_17 {
   class=MaterialExpressionLinearInterpolate
   inputs [
    A {from=""}
    B {from=MaterialExpressionNamedRerouteDeclaration_0}
    Alpha {from=MaterialExpressionNamedRerouteUsage_72}
   ]
  }
  MaterialExpressionLinearInterpolate_18 {
   class=MaterialExpressionLinearInterpolate
   inputs [
    A {from=MaterialExpressionLinearInterpolate_16}
    B {from=MaterialExpressionNamedRerouteUsage_10}
    Alpha {from=MaterialExpressionMax_1}
   ]
  }
  MaterialExpressionLinearInterpolate_2 {
   class=MaterialExpressionLinearInterpolate
   inputs [
    A {from=MaterialExpressionConvert_18}
    B {from=MaterialExpressionAppendVector_2}
    Alpha {from=MaterialExpressionNamedRerouteUsage_7}
   ]
   properties {constA=1}
  }
  MaterialExpressionLinearInterpolate_3 {
   class=MaterialExpressionLinearInterpolate
   inputs [
    A {from=MaterialExpressionConvert_15}
    B {from=MaterialExpressionConvert_17}
    Alpha {from=MaterialExpressionNamedRerouteUsage_13}
   ]
  }
  MaterialExpressionLinearInterpolate_4 {
   class=MaterialExpressionLinearInterpolate
   inputs [
    A {from=""}
    B {from=MaterialExpressionMultiply_13}
    Alpha {from=MaterialExpressionScalarParameter_9}
   ]
   properties {constA=1}
  }
  MaterialExpressionLinearInterpolate_9 {
   class=MaterialExpressionLinearInterpolate
   inputs [
    A {from=MaterialExpressionLinearInterpolate_11}
    B {from=MaterialExpressionNamedRerouteUsage_60}
    Alpha {from=MaterialExpressionNamedRerouteUsage_63}
   ]
  }
  MaterialExpressionLocalPosition_1 {class=MaterialExpressionLocalPosition, inputs [], properties {includedOffsets=ExcludeOffsets, localOrigin=InstancePreSkinning}}
  MaterialExpressionMaterialFunctionCall_0 {
   class=MaterialExpressionMaterialFunctionCall
   inputs [
    Overide Texture Set UVs {from=""}
    Override Tech Art Mask UVs {from=""}
    Override Basic Emissive UVs {from=""}
   ]
   properties {materialFunction=&MF_CharacterParent_2023_PreFX}
  }
  MaterialExpressionMaterialFunctionCall_1 {
   class=MaterialExpressionMaterialFunctionCall
   inputs [
    Material {from=MaterialExpressionBlendMaterialAttributes_2}
    M {from=MaterialExpressionMaterialFunctionCall_0, output=M}
    Pre Skinned Local Position {from=MaterialExpressionMaterialFunctionCall_0, output=M}
    Distance Blend {from=MaterialExpressionMaterialFunctionCall_0, output=M}
    Fresnel {from=MaterialExpressionMaterialFunctionCall_0, output=M}
   ]
   properties {materialFunction=&MF_CharacterParent_2023_PostFX}
  }
  MaterialExpressionMaterialFunctionCall_10 {
   class=MaterialExpressionMaterialFunctionCall
   inputs [
    UV in {from=MaterialExpressionNamedRerouteUsage_27}
    Rear UV Min {from=MaterialExpressionAppendVector_6}
    Rear UV Max {from=MaterialExpressionAppendVector_7}
    Neck UV Min {from=MaterialExpressionAppendVector_0}
    Neck UV Max {from=MaterialExpressionAppendVector_1}
    Strap V Size {from=MaterialExpressionScalarParameter_2}
   ]
   properties {materialFunction=/BRCosmetics/Characters/Player/Female/Medium/Bodies/F_MED_FermiumStraw/Materials/MaterialFunctions/MF_NormalizeUVIsland_Rear}
  }
  MaterialExpressionMaterialFunctionCall_12 {
   class=MaterialExpressionMaterialFunctionCall
   inputs [Multiplier {from=MaterialExpressionConstant_2}]
   properties {materialFunction=&Pi}
  }
  MaterialExpressionMaterialFunctionCall_15 {
   class=MaterialExpressionMaterialFunctionCall
   inputs [
    UV in {from=MaterialExpressionNamedRerouteUsage_22}
    Island UV Min {from=MaterialExpressionAppendVector_4}
    Island UV Max {from=MaterialExpressionAppendVector_5}
   ]
   properties {materialFunction=/BRCosmetics/Characters/Player/Female/Medium/Bodies/F_MED_FermiumStraw/Materials/MaterialFunctions/MF_NormalizeUVIsland}
  }
  MaterialExpressionMaterialFunctionCall_16 {
   class=MaterialExpressionMaterialFunctionCall
   inputs [
    UV in {from=MaterialExpressionNamedRerouteUsage_46}
    Island UV Min {from=MaterialExpressionAppendVector_10}
    Island UV Max {from=MaterialExpressionAppendVector_11}
   ]
   properties {materialFunction=&MF_NormalizeUVIsland}
  }
  MaterialExpressionMaterialFunctionCall_18 {
   class=MaterialExpressionMaterialFunctionCall
   inputs [
    UV in {from=MaterialExpressionNamedRerouteUsage_38}
    Island UV Min {from=MaterialExpressionAppendVector_8}
    Island UV Max {from=MaterialExpressionAppendVector_9}
   ]
   properties {materialFunction=&MF_NormalizeUVIsland}
  }
  MaterialExpressionMaterialFunctionCall_8 {
   class=MaterialExpressionMaterialFunctionCall
   inputs [
    UV Index {from=""}
    Texture {from=""}
    WPO Scale {from=""}
   ]
   properties {materialFunction=/BRCosmetics/Characters/Player/Female/Medium/Bodies/F_MED_FermiumStraw/Materials/MaterialFunctions/DEV/MF_UVTileDebug}
  }
  MaterialExpressionMax_0 {
   class=MaterialExpressionMax
   inputs [
    A {from=MaterialExpressionChannelMaskParameter_8, output=R}
    B {from=MaterialExpressionChannelMaskParameter_7, output=R}
   ]
  }
  MaterialExpressionMax_1 {
   class=MaterialExpressionMax
   inputs [
    A {from=MaterialExpressionNamedRerouteUsage_14}
    B {from=MaterialExpressionNamedRerouteUsage_79}
   ]
  }
  MaterialExpressionMultiply_0 {
   class=MaterialExpressionMultiply
   inputs [
    A {from=MaterialExpressionScalarParameter_0}
    B {from=""}
   ]
   properties {constB=0}
  }
  MaterialExpressionMultiply_1 {
   class=MaterialExpressionMultiply
   inputs [
    A {from=MaterialExpressionConvert_10}
    B {from=MaterialExpressionNamedRerouteUsage_12}
   ]
  }
  MaterialExpressionMultiply_10 {
   class=MaterialExpressionMultiply
   inputs [
    A {from=MaterialExpressionOneMinus_3}
    B {from=MaterialExpressionOneMinus_4}
   ]
  }
  MaterialExpressionMultiply_11 {
   class=MaterialExpressionMultiply
   inputs [
    A {from=MaterialExpressionMultiply_14}
    B {from=MaterialExpressionNamedRerouteUsage_23}
   ]
  }
  MaterialExpressionMultiply_12 {
   class=MaterialExpressionMultiply
   inputs [
    A {from=MaterialExpressionTextureSampleParameter2D_0, output=G}
    B {from=MaterialExpressionScalarParameter_5}
   ]
  }
  MaterialExpressionMultiply_13 {
   class=MaterialExpressionMultiply
   inputs [
    A {from=MaterialExpressionSphereMask_30}
    B {from=MaterialExpressionSphereMask_29}
   ]
  }
  MaterialExpressionMultiply_14 {
   class=MaterialExpressionMultiply
   inputs [
    A {from=MaterialExpressionMultiply_10}
    B {from=MaterialExpressionOneMinus_5}
   ]
  }
  MaterialExpressionMultiply_15 {
   class=MaterialExpressionMultiply
   inputs [
    A {from=MaterialExpressionNamedRerouteUsage_64}
    B {from=MaterialExpressionVertexInterpolator_1, output=PS}
   ]
  }
  MaterialExpressionMultiply_16 {
   class=MaterialExpressionMultiply
   inputs [
    A {from=MaterialExpressionNamedRerouteUsage_64}
    B {from=MaterialExpressionOneMinus_1}
   ]
  }
  MaterialExpressionMultiply_17 {
   class=MaterialExpressionMultiply
   inputs [
    A {from=MaterialExpressionMaterialFunctionCall_12, output=Result}
    B {from=MaterialExpressionNamedRerouteUsage_85}
   ]
  }
  MaterialExpressionMultiply_19 {
   class=MaterialExpressionMultiply
   inputs [
    A {from=MaterialExpressionNamedRerouteUsage_90}
    B {from=MaterialExpressionConstant2Vector_5}
   ]
   properties {desc=aspect ratio}
  }
  MaterialExpressionMultiply_2 {
   class=MaterialExpressionMultiply
   inputs [
    A {from=MaterialExpressionNamedRerouteDeclaration_21}
    B {from=MaterialExpressionOneMinus_0}
   ]
  }
  MaterialExpressionMultiply_23 {
   class=MaterialExpressionMultiply
   inputs [
    A {from=MaterialExpressionNamedRerouteUsage_19}
    B {from=""}
   ]
   properties {constB=-1}
  }
  MaterialExpressionMultiply_25 {
   class=MaterialExpressionMultiply
   inputs [
    A {from=MaterialExpressionConvert_21}
    B {from=MaterialExpressionNamedRerouteUsage_29}
   ]
  }
  MaterialExpressionMultiply_3 {
   class=MaterialExpressionMultiply
   inputs [
    A {from=MaterialExpressionConvert_14}
    B {from=MaterialExpressionScalarParameter_12}
   ]
  }
  MaterialExpressionMultiply_4 {
   class=MaterialExpressionMultiply
   inputs [
    A {from=MaterialExpressionNamedRerouteDeclaration_5}
    B {from=MaterialExpressionNamedRerouteUsage_16}
   ]
  }
  MaterialExpressionMultiply_5 {
   class=MaterialExpressionMultiply
   inputs [
    A {from=MaterialExpressionMultiply_1}
    B {from=""}
   ]
   properties {constB=0.01}
  }
  MaterialExpressionMultiply_7 {
   class=MaterialExpressionMultiply
   inputs [
    A {from=MaterialExpressionChannelMaskParameter_1, output=R}
    B {from=""}
   ]
   properties {constB=5}
  }
  MaterialExpressionMultiply_8 {
   class=MaterialExpressionMultiply
   inputs [
    A {from=MaterialExpressionPower_0}
    B {from=MaterialExpressionVectorParameter_9, output=RGB}
   ]
  }
  MaterialExpressionMultiply_9 {
   class=MaterialExpressionMultiply
   inputs [
    A {from=MaterialExpressionLinearInterpolate_0}
    B {from=MaterialExpressionComponentMask_9}
   ]
  }
  MaterialExpressionNamedRerouteDeclaration_0 {
   class=MaterialExpressionNamedRerouteDeclaration
   inputs [None {from=MaterialExpressionConvert_8}]
   properties {name="Front UV Transforms (V4)", nodeColor=<rgba 1 0.6563 0 1>}
  }
  MaterialExpressionNamedRerouteDeclaration_1 {
   class=MaterialExpressionNamedRerouteDeclaration
   inputs [None {from=MaterialExpressionTextureCoordinate_6}]
   properties {name="UV Raw (V2)", nodeColor=<rgba 1 0 0.7969 1>}
  }
  MaterialExpressionNamedRerouteDeclaration_11 {
   class=MaterialExpressionNamedRerouteDeclaration
   inputs [None {from=MaterialExpressionConvert_16}]
   properties {name="Sleeves UV Transforms (V4)", nodeColor=<rgba 1 0.6562 0 1>}
  }
  MaterialExpressionNamedRerouteDeclaration_12 {
   class=MaterialExpressionNamedRerouteDeclaration
   inputs [None {from=MaterialExpressionSaturate_1}]
   properties {name="Enable User Transforms (V3)", nodeColor=<rgba 1 0 0.3047 1>}
  }
  MaterialExpressionNamedRerouteDeclaration_13 {
   class=MaterialExpressionNamedRerouteDeclaration
   inputs [None {from=MaterialExpressionMaterialFunctionCall_0, output=Result}]
   properties {name="Material In (MA)", nodeColor=<rgba 0.4531 1 0 1>}
  }
  MaterialExpressionNamedRerouteDeclaration_14 {
   class=MaterialExpressionNamedRerouteDeclaration
   inputs [None {from=MaterialExpressionMaterialFunctionCall_0, output=Tech Art Mask}]
   properties {name="TechArt Mask (V3)", nodeColor=<rgba 0.5703 0 1 1>}
  }
  MaterialExpressionNamedRerouteDeclaration_15 {
   class=MaterialExpressionNamedRerouteDeclaration
   inputs [None {from=MaterialExpressionChannelMaskParameter_4, output=R}]
   properties {name="Projection - Sleeves Mask (S)", nodeColor=<rgba 0.3828 0 1 1>}
  }
  MaterialExpressionNamedRerouteDeclaration_16 {
   class=MaterialExpressionNamedRerouteDeclaration
   inputs [None {from=MaterialExpressionMax_0}]
   properties {name="Projection - Rear Mask (S)", nodeColor=<rgba 0.3828 0 1 1>}
  }
  MaterialExpressionNamedRerouteDeclaration_17 {
   class=MaterialExpressionNamedRerouteDeclaration
   inputs [None {from=MaterialExpressionMaterialFunctionCall_15, output="Island UV - 1:1 Aspect Ratio (V2)"}]
   properties {name="UV Front (V2)", nodeColor=<rgba 0 1 0.8828 1>}
  }
  MaterialExpressionNamedRerouteDeclaration_18 {
   class=MaterialExpressionNamedRerouteDeclaration
   inputs [None {from=MaterialExpressionMultiply_2}]
   properties {name="Secondary Parts Mask (S)", nodeColor=<rgba 0.3828 0 1 1>}
  }
  MaterialExpressionNamedRerouteDeclaration_19 {
   class=MaterialExpressionNamedRerouteDeclaration
   inputs [None {from=MaterialExpressionLinearInterpolate_12}]
   properties {name="Outfit Color (V3)", nodeColor=<rgba 0 1 0.7891 1>}
  }
  MaterialExpressionNamedRerouteDeclaration_2 {
   class=MaterialExpressionNamedRerouteDeclaration
   inputs [None {from=MaterialExpressionSaturate_0}]
   properties {name="Force Symmetric Sleeves (S)", nodeColor=<rgba 1 0.9375 0 1>}
  }
  MaterialExpressionNamedRerouteDeclaration_20 {
   class=MaterialExpressionNamedRerouteDeclaration
   inputs [None {from=MaterialExpressionReroute_0}]
   properties {name="Fabric Mask (S)", nodeColor=<rgba 0.3828 0 1 1>}
  }
  MaterialExpressionNamedRerouteDeclaration_21 {
   class=MaterialExpressionNamedRerouteDeclaration
   inputs [None {from=MaterialExpressionSaturate_4}]
   properties {name="Customization Mask (S)", nodeColor=<rgba 0.3828 0 1 1>}
  }
  MaterialExpressionNamedRerouteDeclaration_22 {
   class=MaterialExpressionNamedRerouteDeclaration
   inputs [None {from=MaterialExpressionMaterialFunctionCall_10, output="Island UV - 1:1 Aspect Ratio (V2)"}]
   properties {name="UV Rear (V2)", nodeColor=<rgba 0 1 0.8828 1>}
  }
  MaterialExpressionNamedRerouteDeclaration_23 {
   class=MaterialExpressionNamedRerouteDeclaration
   inputs [None {from=MaterialExpressionMaterialFunctionCall_18, output="Island UV - 1:1 Aspect Ratio (V2)"}]
   properties {name="UV Sleeve Left (V2)", nodeColor=<rgba 0 1 0.8828 1>}
  }
  MaterialExpressionNamedRerouteDeclaration_24 {
   class=MaterialExpressionNamedRerouteDeclaration
   inputs [None {from=MaterialExpressionMultiply_11}]
   properties {name="Projection - Front Mask (S)", nodeColor=<rgba 0.3828 0 1 1>}
  }
  MaterialExpressionNamedRerouteDeclaration_26 {
   class=MaterialExpressionNamedRerouteDeclaration
   inputs [None {from=MaterialExpressionAdd_2}]
   properties {name="Loading Screen Mask (S)", nodeColor=<rgba 1 0.1406 0 1>}
  }
  MaterialExpressionNamedRerouteDeclaration_27 {
   class=MaterialExpressionNamedRerouteDeclaration
   inputs [None {from=MaterialExpressionStaticSwitchParameter_2}]
   properties {name="Loading Screen Color (V3)", nodeColor=<rgba 0 0.1797 1 1>}
  }
  MaterialExpressionNamedRerouteDeclaration_28 {
   class=MaterialExpressionNamedRerouteDeclaration
   inputs [None {from=MaterialExpressionMultiply_8}]
   properties {name=Background Color, nodeColor=<rgba 0.5 1 0 1>}
  }
  MaterialExpressionNamedRerouteDeclaration_29 {
   class=MaterialExpressionNamedRerouteDeclaration
   inputs [None {from=MaterialExpressionConvert_12}]
   properties {name="User - Front Offset (S)", nodeColor=<rgba 0 0.625 1 1>}
  }
  MaterialExpressionNamedRerouteDeclaration_3 {
   class=MaterialExpressionNamedRerouteDeclaration
   inputs [None {from=MaterialExpressionConvert_11}]
   properties {name="U Offset (S)", nodeColor=<rgba 0.6172 0 1 1>}
  }
  MaterialExpressionNamedRerouteDeclaration_30 {
   class=MaterialExpressionNamedRerouteDeclaration
   inputs [None {from=MaterialExpressionGetMaterialAttributes_0, output=Base Color}]
   properties {name="BaseColor In (V3)", nodeColor=<rgba 0.6875 1 0 1>}
  }
  MaterialExpressionNamedRerouteDeclaration_32 {
   class=MaterialExpressionNamedRerouteDeclaration
   inputs [None {from=MaterialExpressionMaterialFunctionCall_16, output="Island UV - 1:1 Aspect Ratio (V2)"}]
   properties {name="UV Sleeve Right (V2)", nodeColor=<rgba 0 1 0.8828 1>}
  }
  MaterialExpressionNamedRerouteDeclaration_33 {
   class=MaterialExpressionNamedRerouteDeclaration
   inputs [None {from=MaterialExpressionMultiply_15}]
   properties {name="Projection - Sleeve Left Mask (S)", nodeColor=<rgba 0.2656 0 1 1>}
  }
  MaterialExpressionNamedRerouteDeclaration_34 {
   class=MaterialExpressionNamedRerouteDeclaration
   inputs [None {from=MaterialExpressionMultiply_16}]
   properties {name="Projection - Sleeve Right Mask (S)", nodeColor=<rgba 0.2656 0 1 1>}
  }
  MaterialExpressionNamedRerouteDeclaration_35 {
   class=MaterialExpressionNamedRerouteDeclaration
   inputs [None {from=MaterialExpressionLinearInterpolate_14}]
   properties {name="Base UV (V2)", nodeColor=<rgba 0 0.1562 1 1>}
  }
  MaterialExpressionNamedRerouteDeclaration_36 {
   class=MaterialExpressionNamedRerouteDeclaration
   inputs [None {from=MaterialExpressionConvert_12}]
   properties {name="User - Back Offset (S)", nodeColor=<rgba 0 0.625 1 1>}
  }
  MaterialExpressionNamedRerouteDeclaration_37 {
   class=MaterialExpressionNamedRerouteDeclaration
   inputs [None {from=MaterialExpressionReroute_20}]
   properties {name="UV Offset (V2)", nodeColor=<rgba 0.5234 0 1 1>}
  }
  MaterialExpressionNamedRerouteDeclaration_38 {
   class=MaterialExpressionNamedRerouteDeclaration
   inputs [None {from=MaterialExpressionReroute_19}]
   properties {name="UV Rotation (S)", nodeColor=<rgba 0 0.5781 1 1>}
  }
  MaterialExpressionNamedRerouteDeclaration_39 {
   class=MaterialExpressionNamedRerouteDeclaration
   inputs [None {from=MaterialExpressionMultiply_25}]
   properties {name="UV Scale (V2)", nodeColor=<rgba 0.3828 0 1 1>}
  }
  MaterialExpressionNamedRerouteDeclaration_4 {
   class=MaterialExpressionNamedRerouteDeclaration
   inputs [None {from=MaterialExpressionLinearInterpolate_4}]
   properties {name="UV Clamp Mask (S)", nodeColor=<rgba 0 0.4375 1 1>}
  }
  MaterialExpressionNamedRerouteDeclaration_40 {
   class=MaterialExpressionNamedRerouteDeclaration
   inputs [None {from=MaterialExpressionConvert_12}]
   properties {name="User - Sleeves Offset (S)", nodeColor=<rgba 0 0.625 1 1>}
  }
  MaterialExpressionNamedRerouteDeclaration_41 {
   class=MaterialExpressionNamedRerouteDeclaration
   inputs [None {from=MaterialExpressionMultiply_4}]
   properties {name="LoadingScreen UV (V2)", nodeColor=<rgba 0 0.1562 1 1>}
  }
  MaterialExpressionNamedRerouteDeclaration_5 {
   class=MaterialExpressionNamedRerouteDeclaration
   inputs [None {from=MaterialExpressionSubtract_17}]
   properties {name="Transformed UV (V2)", nodeColor=<rgba 0.7578 1 0 1>}
  }
  MaterialExpressionNamedRerouteDeclaration_6 {
   class=MaterialExpressionNamedRerouteDeclaration
   inputs [None {from=MaterialExpressionConvert_11}]
   properties {name="V Offset (S)", nodeColor=<rgba 0.6172 0 1 1>}
  }
  MaterialExpressionNamedRerouteDeclaration_7 {
   class=MaterialExpressionNamedRerouteDeclaration
   inputs [None {from=MaterialExpressionConvert_9}]
   properties {name="Back UV Transforms (V4)", nodeColor=<rgba 1 0.6562 0 1>}
  }
  MaterialExpressionNamedRerouteDeclaration_8 {
   class=MaterialExpressionNamedRerouteDeclaration
   inputs [None {from=MaterialExpressionConvert_11}]
   properties {name="UV Scale (S)", nodeColor=<rgba 0.3828 0 1 1>}
  }
  MaterialExpressionNamedRerouteDeclaration_9 {
   class=MaterialExpressionNamedRerouteDeclaration
   inputs [None {from=MaterialExpressionConvert_11}]
   properties {name="UV Base Rotation (S)", nodeColor=<rgba 0 0.5781 1 1>}
  }
  MaterialExpressionNamedRerouteUsage_1 {class=MaterialExpressionNamedRerouteUsage}
  MaterialExpressionNamedRerouteUsage_10 {class=MaterialExpressionNamedRerouteUsage}
  MaterialExpressionNamedRerouteUsage_12 {class=MaterialExpressionNamedRerouteUsage}
  MaterialExpressionNamedRerouteUsage_13 {class=MaterialExpressionNamedRerouteUsage}
  MaterialExpressionNamedRerouteUsage_14 {class=MaterialExpressionNamedRerouteUsage}
  MaterialExpressionNamedRerouteUsage_15 {class=MaterialExpressionNamedRerouteUsage}
  MaterialExpressionNamedRerouteUsage_16 {class=MaterialExpressionNamedRerouteUsage}
  MaterialExpressionNamedRerouteUsage_17 {class=MaterialExpressionNamedRerouteUsage}
  MaterialExpressionNamedRerouteUsage_18 {class=MaterialExpressionNamedRerouteUsage}
  MaterialExpressionNamedRerouteUsage_19 {class=MaterialExpressionNamedRerouteUsage}
  MaterialExpressionNamedRerouteUsage_20 {class=MaterialExpressionNamedRerouteUsage}
  MaterialExpressionNamedRerouteUsage_21 {class=MaterialExpressionNamedRerouteUsage}
  MaterialExpressionNamedRerouteUsage_22 {class=MaterialExpressionNamedRerouteUsage}
  MaterialExpressionNamedRerouteUsage_23 {class=MaterialExpressionNamedRerouteUsage}
  MaterialExpressionNamedRerouteUsage_24 {class=MaterialExpressionNamedRerouteUsage}
  MaterialExpressionNamedRerouteUsage_25 {class=MaterialExpressionNamedRerouteUsage}
  MaterialExpressionNamedRerouteUsage_26 {class=MaterialExpressionNamedRerouteUsage}
  MaterialExpressionNamedRerouteUsage_27 {class=MaterialExpressionNamedRerouteUsage}
  MaterialExpressionNamedRerouteUsage_28 {class=MaterialExpressionNamedRerouteUsage}
  MaterialExpressionNamedRerouteUsage_29 {class=MaterialExpressionNamedRerouteUsage}
  MaterialExpressionNamedRerouteUsage_30 {class=MaterialExpressionNamedRerouteUsage}
  MaterialExpressionNamedRerouteUsage_31 {class=MaterialExpressionNamedRerouteUsage}
  MaterialExpressionNamedRerouteUsage_32 {class=MaterialExpressionNamedRerouteUsage}
  MaterialExpressionNamedRerouteUsage_33 {class=MaterialExpressionNamedRerouteUsage}
  MaterialExpressionNamedRerouteUsage_38 {class=MaterialExpressionNamedRerouteUsage}
  MaterialExpressionNamedRerouteUsage_43 {class=MaterialExpressionNamedRerouteUsage}
  MaterialExpressionNamedRerouteUsage_46 {class=MaterialExpressionNamedRerouteUsage}
  MaterialExpressionNamedRerouteUsage_47 {class=MaterialExpressionNamedRerouteUsage}
  MaterialExpressionNamedRerouteUsage_48 {class=MaterialExpressionNamedRerouteUsage}
  MaterialExpressionNamedRerouteUsage_49 {class=MaterialExpressionNamedRerouteUsage}
  MaterialExpressionNamedRerouteUsage_51 {class=MaterialExpressionNamedRerouteUsage}
  MaterialExpressionNamedRerouteUsage_53 {class=MaterialExpressionNamedRerouteUsage}
  MaterialExpressionNamedRerouteUsage_57 {class=MaterialExpressionNamedRerouteUsage}
  MaterialExpressionNamedRerouteUsage_59 {class=MaterialExpressionNamedRerouteUsage}
  MaterialExpressionNamedRerouteUsage_6 {class=MaterialExpressionNamedRerouteUsage}
  MaterialExpressionNamedRerouteUsage_60 {class=MaterialExpressionNamedRerouteUsage}
  MaterialExpressionNamedRerouteUsage_61 {class=MaterialExpressionNamedRerouteUsage}
  MaterialExpressionNamedRerouteUsage_62 {class=MaterialExpressionNamedRerouteUsage}
  MaterialExpressionNamedRerouteUsage_63 {class=MaterialExpressionNamedRerouteUsage}
  MaterialExpressionNamedRerouteUsage_64 {class=MaterialExpressionNamedRerouteUsage}
  MaterialExpressionNamedRerouteUsage_69 {class=MaterialExpressionNamedRerouteUsage}
  MaterialExpressionNamedRerouteUsage_7 {class=MaterialExpressionNamedRerouteUsage}
  MaterialExpressionNamedRerouteUsage_70 {class=MaterialExpressionNamedRerouteUsage}
  MaterialExpressionNamedRerouteUsage_72 {class=MaterialExpressionNamedRerouteUsage}
  MaterialExpressionNamedRerouteUsage_77 {class=MaterialExpressionNamedRerouteUsage}
  MaterialExpressionNamedRerouteUsage_79 {class=MaterialExpressionNamedRerouteUsage}
  MaterialExpressionNamedRerouteUsage_8 {class=MaterialExpressionNamedRerouteUsage}
  MaterialExpressionNamedRerouteUsage_85 {class=MaterialExpressionNamedRerouteUsage}
  MaterialExpressionNamedRerouteUsage_88 {class=MaterialExpressionNamedRerouteUsage}
  MaterialExpressionNamedRerouteUsage_89 {class=MaterialExpressionNamedRerouteUsage}
  MaterialExpressionNamedRerouteUsage_9 {class=MaterialExpressionNamedRerouteUsage}
  MaterialExpressionNamedRerouteUsage_90 {class=MaterialExpressionNamedRerouteUsage}
  MaterialExpressionNamedRerouteUsage_91 {class=MaterialExpressionNamedRerouteUsage}
  MaterialExpressionOneMinus_0 {class=MaterialExpressionOneMinus, inputs [None {from=MaterialExpressionReroute_0}]}
  MaterialExpressionOneMinus_1 {class=MaterialExpressionOneMinus, inputs [None {from=MaterialExpressionVertexInterpolator_1, output=PS}]}
  MaterialExpressionOneMinus_3 {class=MaterialExpressionOneMinus, inputs [None {from=MaterialExpressionNamedRerouteUsage_26}]}
  MaterialExpressionOneMinus_4 {class=MaterialExpressionOneMinus, inputs [None {from=MaterialExpressionNamedRerouteUsage_25}]}
  MaterialExpressionOneMinus_5 {class=MaterialExpressionOneMinus, inputs [None {from=MaterialExpressionChannelMaskParameter_9, output=R}]}
  MaterialExpressionOneMinus_9 {class=MaterialExpressionOneMinus, inputs [None {from=MaterialExpressionNamedRerouteUsage_31}]}
  MaterialExpressionPower_0 {
   class=MaterialExpressionPower
   inputs [
    Base {from=MaterialExpressionNamedRerouteDeclaration_30}
    Exp {from=MaterialExpressionScalarParameter_4}
   ]
   properties {constExponent=0.25}
  }
  MaterialExpressionReroute_0 {class=MaterialExpressionReroute, inputs [None {from=MaterialExpressionChannelMaskParameter_2, output=R}]}
  MaterialExpressionReroute_13 {class=MaterialExpressionReroute, inputs [None {from=MaterialExpressionMaterialFunctionCall_1, output=Result}]}
  MaterialExpressionReroute_14 {class=MaterialExpressionReroute, inputs [None {from=MaterialExpressionMaterialFunctionCall_8, output="WPO (V3)"}]}
  MaterialExpressionReroute_17 {class=MaterialExpressionReroute, inputs [None {from=MaterialExpressionVectorParameter_10, output=A}]}
  MaterialExpressionReroute_18 {class=MaterialExpressionReroute, inputs [None {from=MaterialExpressionReroute_17}]}
  MaterialExpressionReroute_19 {class=MaterialExpressionReroute, inputs [None {from=MaterialExpressionConvert_21}]}
  MaterialExpressionReroute_20 {class=MaterialExpressionReroute, inputs [None {from=MaterialExpressionConvert_21}]}
  MaterialExpressionReroute_5 {class=MaterialExpressionReroute, inputs [None {from=MaterialExpressionSetMaterialAttributes_4}]}
  MaterialExpressionReroute_7 {class=MaterialExpressionReroute, inputs [None {from=MaterialExpressionAdd_7}]}
  MaterialExpressionRotator_2 {
   class=MaterialExpressionRotator
   inputs [
    Coordinate {from=MaterialExpressionSubtract_16, output=""}
    Time {from=MaterialExpressionMultiply_17, output=""}
   ]
   properties {centerX=0, centerY=0, speed=-1}
  }
  MaterialExpressionRound_0 {class=MaterialExpressionRound, inputs [None {from=MaterialExpressionScalarParameter_8, output=""}], properties {}}
  MaterialExpressionRound_1 {class=MaterialExpressionRound, inputs [None {from=MaterialExpressionMultiply_3, output=""}], properties {}}
  MaterialExpressionSaturate_0 {class=MaterialExpressionSaturate, inputs [None {from=MaterialExpressionRound_0}]}
  MaterialExpressionSaturate_1 {class=MaterialExpressionSaturate, inputs [None {from=MaterialExpressionRound_1}]}
  MaterialExpressionSaturate_4 {class=MaterialExpressionSaturate, inputs [None {from=MaterialExpressionMultiply_7}]}
  MaterialExpressionSaturate_5 {class=MaterialExpressionSaturate, inputs [None {from=MaterialExpressionSign_1}]}
  MaterialExpressionScalarParameter_0 {
   class=MaterialExpressionScalarParameter
   properties {group=FermiumStraw | 03_DEV, parameterName=Enable Base Transforms, sliderMax=1, sortPriority=10}
  }
  MaterialExpressionScalarParameter_1 {
   class=MaterialExpressionScalarParameter
   properties {group=FermiumStraw | 03_DEV, parameterName=Enable Debug Color, sliderMax=1, sortPriority=40}
  }
  MaterialExpressionScalarParameter_11 {
   class=MaterialExpressionScalarParameter
   properties {group=FermiumStraw | 02_Transforms | 01_Base Transforms, parameterName=LoadingScreen_UVTransform_LeftSleeveUOffset, sortPriority=40}
  }
  MaterialExpressionScalarParameter_12 {class=MaterialExpressionScalarParameter, properties {group=FermiumStraw | 03_DEV, parameterName=Enable User Transforms, sortPriority=20}}
  MaterialExpressionScalarParameter_13 {
   class=MaterialExpressionScalarParameter
   properties {group=FermiumStraw | 02_Transforms | 02_User Transforms, parameterName=LoadingScreen_UVTransform_UserFrontOffset, sortPriority=110}
  }
  MaterialExpressionScalarParameter_14 {
   class=MaterialExpressionScalarParameter
   properties {group=FermiumStraw | 02_Transforms | 02_User Transforms, parameterName=LoadingScreen_UVTransform_UserBackOffset, sortPriority=120}
  }
  MaterialExpressionScalarParameter_15 {
   class=MaterialExpressionScalarParameter
   properties {group=FermiumStraw | 02_Transforms | 02_User Transforms, parameterName=LoadingScreen_UVTransform_UserSleeveOffset, sortPriority=130}
  }
  MaterialExpressionScalarParameter_18 {
   class=MaterialExpressionScalarParameter
   properties {
    defaultValue=1
    group=FermiumStraw | 02_Transforms | 02_User Transforms
    parameterName=LoadingScreen_UVTransform_EnableUserFrontOffset
    sortPriority=10
   }
  }
  MaterialExpressionScalarParameter_19 {
   class=MaterialExpressionScalarParameter
   properties {
    defaultValue=1
    group=FermiumStraw | 02_Transforms | 02_User Transforms
    parameterName=LoadingScreen_UVTransform_EnableUserBackOffset
    sortPriority=20
   }
  }
  MaterialExpressionScalarParameter_2 {
   class=MaterialExpressionScalarParameter
   properties {defaultValue=0.031, group=FermiumStraw | 01_Appearance, parameterName=VSize_Strap, sliderMax=1, sortPriority=312}
  }
  MaterialExpressionScalarParameter_20 {
   class=MaterialExpressionScalarParameter
   properties {
    defaultValue=1
    group=FermiumStraw | 02_Transforms | 02_User Transforms
    parameterName=LoadingScreen_UVTransform_EnableUserSleevesOffset
    sortPriority=30
   }
  }
  MaterialExpressionScalarParameter_3 {
   class=MaterialExpressionScalarParameter
   properties {defaultValue=2, group=zz_SIGNAL DEBUG, parameterName=Flattened UV Z Offset, sortPriority=-1490}
  }
  MaterialExpressionScalarParameter_4 {
   class=MaterialExpressionScalarParameter
   properties {
    defaultValue=0.25
    desc=To brighten up the basecolor below the loading screen before multiplying it
    group=FermiumStraw | 01_Appearance
    parameterName=LoadingScreen_BackgroundGamma
    sortPriority=10
   }
  }
  MaterialExpressionScalarParameter_5 {
   class=MaterialExpressionScalarParameter
   properties {defaultValue=1, group=FermiumStraw | 01_Appearance, parameterName=LoadingScreenEnabled, sliderMax=1, sortPriority=100}
  }
  MaterialExpressionScalarParameter_8 {
   class=MaterialExpressionScalarParameter
   properties {group=FermiumStraw | 03_DEV, parameterName=Enable Symmetric Sleeves, sliderMax=1, sortPriority=30}
  }
  MaterialExpressionScalarParameter_9 {class=MaterialExpressionScalarParameter, properties {group=FermiumStraw | 01_Appearance, parameterName=LoadingScreenClamp, sortPriority=140}}
  MaterialExpressionSetMaterialAttributes_0 {
   class=MaterialExpressionSetMaterialAttributes
   inputs [
    MaterialAttributes {from=MaterialExpressionReroute_13, output=""}
    World Position Offset {from=MaterialExpressionAdd_3, output=""}
   ]
   properties {attributeSetTypes [#2]}
  }
  MaterialExpressionSetMaterialAttributes_4 {
   class=MaterialExpressionSetMaterialAttributes
   inputs [
    MaterialAttributes {from=MaterialExpressionNamedRerouteDeclaration_13, output=""}
    Base Color {from=MaterialExpressionNamedRerouteUsage_53, output=""}
   ]
   properties {attributeSetTypes [#1]}
  }
  MaterialExpressionSign_1 {class=MaterialExpressionSign, inputs [None {from=MaterialExpressionComponentMask_5, output=""}], properties {}}
  MaterialExpressionSphereMask_29 {
   class=MaterialExpressionSphereMask
   inputs [
    A {from=MaterialExpressionConstant_11, output=""}
    B {from=MaterialExpressionConvert_13, output=""}
    Radius {from=MaterialExpressionConstant_11, output=""}
    Hardness {from="", output=""}
   ]
   properties {attenuationRadius=0.5}
  }
  MaterialExpressionSphereMask_30 {
   class=MaterialExpressionSphereMask
   inputs [
    A {from=MaterialExpressionConstant_11, output=""}
    B {from=MaterialExpressionConvert_13, output=""}
    Radius {from=MaterialExpressionConstant_11, output=""}
    Hardness {from="", output=""}
   ]
   properties {
    attenuationRadius=0.5
    desc="2 * 1D mask instead of a single 2D mask\r\n=> keeps sharp angles at corners of UV\r\ntile, otherwise mask will be ovale"
   }
  }
  MaterialExpressionStaticSwitchParameter_19 {
   class=MaterialExpressionStaticSwitchParameter
   inputs [
    True {from=MaterialExpressionSetMaterialAttributes_0, output=""}
    False {from=MaterialExpressionMaterialFunctionCall_1, output=Result}
   ]
   properties {group=zz_SIGNAL DEBUG, parameterName=Flatten UV, sortPriority=-1500}
  }
  MaterialExpressionStaticSwitchParameter_2 {
   class=MaterialExpressionStaticSwitchParameter
   inputs [
    True {from=MaterialExpressionTextureSampleParameter2D_2, output=RGB}
    False {from=MaterialExpressionTextureSampleParameter2D_1, output=RGB}
   ]
   properties {group=FermiumStraw | 01_Appearance, parameterName=OverrideLoadingScreen, sortPriority=110}
  }
  MaterialExpressionSubtract_16 {
   class=MaterialExpressionSubtract
   inputs [
    A {from=MaterialExpressionFrac_1}
    B {from=""}
   ]
   properties {constB=0.5}
  }
  MaterialExpressionSubtract_17 {
   class=MaterialExpressionSubtract
   inputs [
    A {from=MaterialExpressionAdd_1}
    B {from=MaterialExpressionNamedRerouteUsage_91}
   ]
  }
  MaterialExpressionSubtract_3 {
   class=MaterialExpressionSubtract
   inputs [
    A {from=MaterialExpressionMultiply_23}
    B {from=MaterialExpressionScalarParameter_11}
   ]
  }
  MaterialExpressionTextureCoordinate_6 {class=MaterialExpressionTextureCoordinate}
  MaterialExpressionTextureSampleParameter2D_0 {
   class=MaterialExpressionTextureSampleParameter2D
   inputs [
    UVs {from="", output=""}
    Bias {from=MaterialExpressionConstant_0, output=""}
    Apply View MipBias {from="", output=""}
   ]
   properties {
    group=FermiumStraw | 01_Appearance
    mipValueMode=TMVM_MipBias
    parameterName=MaskTexture
    samplerSource=SSM_Wrap_WorldGroupSettings
    samplerType=SAMPLERTYPE_Masks
    sortPriority=200
    texture=/BRCosmetics/Characters/Player/Female/Medium/Bodies/F_MED_FermiumStraw/Textures/T_FermiumStraw_Body_FX3
   }
  }
  MaterialExpressionTextureSampleParameter2D_1 {
   class=MaterialExpressionTextureSampleParameter2D
   inputs [
    UVs {from=MaterialExpressionNamedRerouteUsage_18, output=""}
    Apply View MipBias {from="", output=""}
   ]
   properties {
    group=FermiumStraw | 01_Appearance
    parameterName=LoadingScreen
    samplerSource=SSM_Wrap_WorldGroupSettings
    sortPriority=120
    texture=/BRCosmetics/2dAssets/Loadingscreens/Season30/T_LS_S30_GalliumConcrete
   }
  }
  MaterialExpressionTextureSampleParameter2D_2 {
   class=MaterialExpressionTextureSampleParameter2D
   inputs [
    UVs {from=MaterialExpressionNamedRerouteUsage_89, output=""}
    Apply View MipBias {from="", output=""}
   ]
   properties {
    group=FermiumStraw | 01_Appearance
    parameterName=LoadingScreenOverride
    samplerSource=SSM_Wrap_WorldGroupSettings
    sortPriority=130
    texture=&T_LS_S30_GalliumConcrete
   }
  }
  MaterialExpressionTextureSampleParameter2D_3 {
   class=MaterialExpressionTextureSampleParameter2D
   inputs [
    UVs {from="", output=""}
    Apply View MipBias {from="", output=""}
   ]
   properties {
    group=FermiumStraw | 00_Base Masks
    parameterName=ProjectionMask
    samplerSource=SSM_Wrap_WorldGroupSettings
    samplerType=SAMPLERTYPE_Masks
    sortPriority=50
    texture=&T_BlackMask
   }
  }
  MaterialExpressionTextureSampleParameter2D_4 {
   class=MaterialExpressionTextureSampleParameter2D
   inputs [
    UVs {from="", output=""}
    Apply View MipBias {from="", output=""}
   ]
   properties {
    group=FermiumStraw | 00_Base Masks
    parameterName=ExtraMask - TMP
    samplerSource=SSM_Wrap_WorldGroupSettings
    samplerType=SAMPLERTYPE_Masks
    sortPriority=30
    texture=&T_BlackMask
   }
  }
  MaterialExpressionVectorParameter_0 {
   class=MaterialExpressionVectorParameter
   properties {
    group=FermiumStraw | 01_Appearance
    parameterName=UVBounds_Front
    sortPriority=310
    channelNames {a=V Max, b=U Max, g=V Min, r=U Min}
    defaultValue=<rgba 0.05 0.075 0.385 1>
   }
  }
  MaterialExpressionVectorParameter_1 {
   class=MaterialExpressionVectorParameter
   properties {
    group=FermiumStraw | 01_Appearance
    parameterName=UVBounds_Rear
    sortPriority=310
    channelNames {a=V Max, b=U Max, g=V Min, r=U Min}
    defaultValue=<rgba 0.45 0.24 0.795 1>
   }
  }
  MaterialExpressionVectorParameter_10 {
   class=MaterialExpressionVectorParameter
   properties {
    group=FermiumStraw | 02_Transforms | 01_Base Transforms
    parameterName=LoadingScreen_UVTransform_Sleeves
    sortPriority=30
    defaultValue=<rgba 0 0 0 1>
   }
  }
  MaterialExpressionVectorParameter_11 {
   class=MaterialExpressionVectorParameter
   properties {
    group=FermiumStraw | 02_Transforms | 01_Base Transforms
    parameterName=LoadingScreen_UVTransform_Front
    sortPriority=10
    defaultValue=<rgba 0 0 0 1>
   }
  }
  MaterialExpressionVectorParameter_12 {
   class=MaterialExpressionVectorParameter
   properties {
    group=FermiumStraw | 02_Transforms | 01_Base Transforms
    parameterName=LoadingScreen_UVTransform_Back
    sortPriority=20
    defaultValue=<rgba 0 0 0 1>
   }
  }
  MaterialExpressionVectorParameter_2 {
   class=MaterialExpressionVectorParameter
   properties {
    group=FermiumStraw | 01_Appearance
    parameterName=UVBounds_Neck
    sortPriority=310
    channelNames {a=V Max, b=U Max, g=V Min, r=U Min}
    defaultValue=<rgba 0.166 0.015 0.292 0.0875>
   }
  }
  MaterialExpressionVectorParameter_3 {
   class=MaterialExpressionVectorParameter
   properties {group=FermiumStraw | 03_DEV, parameterName=Debug Color, sortPriority=50, defaultValue=<rgba 1 1 1 1>}
  }
  MaterialExpressionVectorParameter_5 {
   class=MaterialExpressionVectorParameter
   properties {
    group=FermiumStraw | 01_Appearance
    parameterName=UVBounds_SleeveLeft
    sortPriority=310
    channelNames {a=V Max, b=U Max, g=V Min, r=U Min}
    defaultValue=<rgba 0.34 0 0.625 0.225>
   }
  }
  MaterialExpressionVectorParameter_7 {
   class=MaterialExpressionVectorParameter
   properties {group=FermiumStraw | 01_Appearance, parameterName=SecondaryColor, sortPriority=30, defaultValue=<rgba 1 1 1 1>}
  }
  MaterialExpressionVectorParameter_8 {
   class=MaterialExpressionVectorParameter
   properties {
    group=FermiumStraw | 01_Appearance
    parameterName=UVBounds_SleeveRight
    sortPriority=310
    channelNames {a=V Max, b=U Max, g=V Min, r=U Min}
    defaultValue=<rgba 0.625 0 0.912 0.225>
   }
  }
  MaterialExpressionVectorParameter_9 {
   class=MaterialExpressionVectorParameter
   properties {group=FermiumStraw | 01_Appearance, parameterName=BackgroundColor, sortPriority=20, defaultValue=<rgba 1 1 1 1>}
  }
  MaterialExpressionVertexInterpolator_1 {class=MaterialExpressionVertexInterpolator, inputs [VS {from=MaterialExpressionSaturate_5, output=""}], properties {}}
 }
 properties {
  bUseMaterialAttributes=true
  bUsedWithSkeletalMesh=true
  bUsedWithStaticMesh=false
  pixelDepthOffsetMode=PDOM_Legacy
  shadingModel=MSM_FromMaterialExpression
  usedShadingModels=Default Lit | Subsurface | Clear Coat | Subsurface Profile | Cloth
 }
}
